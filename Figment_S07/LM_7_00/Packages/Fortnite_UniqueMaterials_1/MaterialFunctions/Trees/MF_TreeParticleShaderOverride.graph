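# Material function that adapts tree material attributes for use on particles, passing them through unchanged unless the IsParticleShader? static switch is on.
# A scaled, optionally panning noise texture is smoothstepped against a threshold from inverted particle alpha, creating an erosion-style fade.
# The result, optionally masked by a selectable texture channel, multiplies the OpacityMask input to override opacity mask.
# The Normal input then replaces the material's normals.
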MaterialFunction /Figment_S07/LM_7_00/Packages/Fortnite_UniqueMaterials_1/MaterialFunctions/Trees/MF_TreeParticleShaderOverride
graph {nodes [
  n1 = FunctionInput() {inputType=FunctionInput_MaterialAttributes}
  n2 = ScalarParameter() {parameterName=Smooth Step U Speed}
  n3 = Time()
  n4 = MaterialFunctionCall(Float2=TextureCoordinate()) {materialFunction=&BreakOutFloat2Components}
  n5 = Multiply(A=ScalarParameter() {defaultValue=1, parameterName=Smooth Step U Scale}, B=@n4.R)
  n6 = Multiply(A=@n4.G, B=ScalarParameter() {defaultValue=1, parameterName=Smooth Step V Scale})
  n7 = ScalarParameter() {parameterName=Smooth Step V Speed}
  n8 = AppendVector(A=Add(A=Multiply(A=@n2, B=@n3), B=@n5), B=Add(A=@n6, B=Multiply(A=@n3, B=@n7)))
  n9 = AppendVector(A=@n5, B=@n6)
  n10 = StaticBoolParameter() {parameterName=Smooth Step Pans}
  n11 = OneMinus(None=Multiply(A=ScalarParameter() {defaultValue=1, parameterName=Smooth Step Constant}, B=ParticleColor().A))
  n12 = MaterialFunctionCall(
   Alpha=TextureSampleParameter2D(UVs=StaticSwitch(True=@n8, False=@n9, Value=@n10)) {parameterName=SmoothStep_Texture, texture=/Figment_S07/LM_7_00/VisualThreatManager/StormVisuals/Textures/2ndInfectedMapLinear}.G
   Min=Add(A=@n11) {constB=-0.1}
   Max=@n11
  ) {materialFunction=&SmoothStep}
  n13 = TextureSampleParameter2D() {parameterName=Opacity Mask, samplerType=SAMPLERTYPE_Masks, texture=/Figment_S07/LM_7_00/Environments/Props/Trees/Textures/Tree_21_M}
  n14 = Multiply(A=@n12.Result, B=StaticComponentMaskParameter(None=@n13.RGB) {defaultR=true, parameterName=Opacity Mask Channel})
  n15 = StaticBoolParameter() {parameterName=Use Opacity Mask}
  n16 = MaterialFunctionCall!(
   Material=@n1
   NewOpacityMask=Multiply(A=StaticSwitch(True=@n14, False=@n12.Result, Value=@n15), B=FunctionInput() {inputName=OpacityMask, sortPriority=2})
   Mask
  ) {materialFunction=/Engine/Functions/MaterialLayerFunctions/MatLayerBlend_OverrideOpacityMask}
  n17 = FunctionInput() {inputName=Normal, sortPriority=3}
  FunctionOutput(None=StaticSwitchParameter(
    True=MaterialFunctionCall!(Material=@n16.Blended Material, Normal=@n17) {materialFunction=/Engine/Functions/MaterialLayerFunctions/MatLayerBlend_ReplaceNormals}.Result
    False=@n1
   ) {group=ParticleShader, parameterName=IsParticleShader?})
 ]}
