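# Material function that adds random color variation to tree elements by tinting the base color of incoming material attributes.
# It builds a lookup coordinate from the actor's doubled world X plus a scaled offset, switched between mesh element index and the B component of Position and Parent Index.
# The sampled Color Change Chart texture multiplies the base color.
# A static switch outputs either this result or the untouched input.
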MaterialFunction /Figment_S07/LM_7_00/Packages/Fortnite_UniqueMaterials_1/MaterialFunctions/Trees/MF_TreeRandomColorElements
graph {nodes [
  n1 = FunctionInput() {inputType=FunctionInput_MaterialAttributes}
  n2 = MaterialFunctionCall(Float3=ActorPositionWS()) {materialFunction=&BreakOutFloat3Components}
  n3 = Add(A=@n2.R, B=@n2.R)
  n4 = ScalarParameter() {defaultValue=0.005, group=PX, parameterName=Color Change Chart Variation Amount}
  n5 = Multiply(A=FunctionInput() {inputName=Mesh Element Index, inputType=FunctionInput_Scalar, sortPriority=5}, B=@n4)
  n6 = Multiply(
   A=@n4
   B=MaterialFunctionCall(Float3=FunctionInput() {inputName=Position and Parent Index, inputType=FunctionInput_Vector4, sortPriority=3}) {materialFunction=&BreakOutFloat3Components}.B
  )
  n7 = TextureSampleParameter2D(UVs=Add(
    A=Multiply(A=@n3) {constB=0.005}
    B=StaticSwitchParameter(True=@n5, False=@n6) {group=PX, parameterName=Randomly Color Elements By Model Index Or Height}
   )) {group=PX, parameterName=Color Change Chart, texture=/Figment_S07/LM_7_00/Vehicles/FORT_Vehicles_01/Materials/CarColors_Interiors}
  n8 = Multiply(
   A=@n7.RGB
   B=MaterialFunctionCall!(Material=@n1) {materialFunction=/Engine/Functions/MaterialLayerFunctions/MatLayerBlend_BreakBaseColor}.BaseColor
  )
  FunctionOutput(None=StaticSwitchParameter(
    True=MaterialFunctionCall!(Material=@n1, NewBaseColor=@n8, Mask) {materialFunction=/Engine/Functions/MaterialLayerFunctions/MatLayerBlend_OverrideBaseColor}.Blended Material
    False=@n1
   ) {group=PX, parameterName=Randomly Color Elements})
 ]}
