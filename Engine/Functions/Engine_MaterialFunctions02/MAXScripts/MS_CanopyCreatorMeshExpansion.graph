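# Material function that widens very thin mesh strips, such as canopy branches, hair, grass or cables, into camera-facing ribbons through World Position Offset.
# It crosses the view vector with the tangent basis Y, then scales by UV0 U centered on 0.5, UV2 radius and a thickening parameter.
# Only zero vertex alpha gets offset.
# It also outputs a normal map transformed into the ribbon basis, branch radius, and alpha-derived thicken and knot masks.
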MaterialFunction /Engine/Functions/Engine_MaterialFunctions02/MAXScripts/MS_CanopyCreatorMeshExpansion
graph {
 properties {
  bExposeToLibrary=true
  description="This function  widens strips of polies along the U of their UVs.  These strips shouuld be extremely thin and imported with remove degenerate triangles disabled. Useful for Hair, Grass, Cables etc."
  libraryCategoriesText [Misc, World Position Offset]
 }
 nodes [
  n1 = FunctionInput() {
   bUsePreviewValueAsDefault=true
   description=Adds a texture normal to the normalmap transform
   inputName=Normal Map
   sortPriority=8
   previewValue=<0 0 1 1>
  }
  n2 = WorldPosition()
  n3 = MaterialFunctionCall() {materialFunction=&TangentBasis}
  n4 = CrossProduct(
   A=Subtract(A=CameraPositionWS(), B=FunctionInput(Preview=@n2.XYZ) {bUsePreviewValueAsDefault=true, inputName=WorldPosition, sortPriority=2})
   B=@n3.Y V
  )
  n5 = MaterialFunctionCall!(Vector1=@n4, Vector2=Multiply(A=@n3.Y V, B=Constant() {r=-1})) {materialFunction=/Engine/Functions/Engine_MaterialFunctions02/Math/CreateThirdOrthogonalVector}
  FunctionOutput(None=MaterialFunctionCall(
    VectorToTransform=@n1
    BasisX=Normalize(VectorInput=@n5.Vector1)
    BasisY=Normalize(VectorInput=@n5.Vector2)
    BasisZ=Normalize(VectorInput=@n5.Vector3)
   ) {materialFunction=&Transform3x3Matrix}.Transformed Vector) {description=Add this result to your other WPO calculations, outputName=Normal}
  n6 = VertexColor()
  n7 = Constant()
  n8 = Constant() {r=1}
  n9 = If(A=@n6.A, B=@n7, "A > B"=@n8, "A == B"=@n7, "A < B"=@n8)
  FunctionOutput(None=@n9) {description=Effect Alpha, outputName="Black=Spline Thicken White=Normal", sortPriority=3}
  n10 = Normalize(VectorInput=@n4)
  n11 = MaterialFunctionCall(Float2=TextureCoordinate() {coordinateIndex=2}) {materialFunction=&BreakOutFloat2Components}
  n12 = ScalarParameter() {defaultValue=1.923, group=BranchThickening, parameterName=Branch Thickening Amount}
  n13 = Constant() {r=-0.5}
  n14 = ComponentMask(None=TextureCoordinate()) {r=true}
  n15 = Multiply(A=Multiply(A=Add(A=@n11.R, B=@n12), B=Add(A=@n13, B=@n14)), B=@n12)
  FunctionOutput(None=Multiply(A=Multiply(A=@n10, B=@n15), B=OneMinus(None=@n9))) {description=Add this result to your other WPO calculations, outputName=WorldPositionOffset, sortPriority=1}
  FunctionOutput(None=@n11.R) {outputName=Branch Radius, sortPriority=4}
  n16 = SphereMask(A=@n6.A, B=Constant() {r=0.5}) {attenuationRadius=0.1}
  FunctionOutput(None=@n16) {description=Effect Alpha, outputName="white = knots black = everything else", sortPriority=3}
  OneMinus(None=@n16)
 ]
}
