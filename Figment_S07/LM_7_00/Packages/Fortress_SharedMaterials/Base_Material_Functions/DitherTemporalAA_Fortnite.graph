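# Material function that dithers an alpha threshold in screen space and over time, for TemporalAA-friendly masked translucency or anisotropic materials.
# A custom HLSL node computes a five-step stipple from pixel coordinates shifted by an Offset input and the TemporalAA sample index.
# Screen-aligned 64x64 tiling noise, with its own offset and scaled by Random, is added to break up the pattern.
# The normalized sum, re-centered by one half, biases the Alpha Threshold output.
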MaterialFunction /Figment_S07/LM_7_00/Packages/Fortress_SharedMaterials/Base_Material_Functions/DitherTemporalAA_Fortnite
graph {
 properties {description="Stipple pattern/Dither  in screen space and time to work well with TemporalAA for things like masked translucency or anisotropic materials"}
 nodes [
  n1 = FunctionInput() {bUsePreviewValueAsDefault=true, inputName=Alpha Threshold, inputType=FunctionInput_Scalar, sortPriority=-1, previewValue=<0.5 0 0 1>}
  n2 = Constant2Vector() {g=1, r=1}
  n3 = Custom!(None) {
   code=View.TemporalAAParams
   desc="\n*** Original source before expression upgrade ***\nView.TemporalAAParams\n*** Original source before expression upgrade ***\nFrame.TemporalAAParams"
   description=TemporalAA SampleID
   outputType=CMOT_Float2
  }
  MaterialExpressionAdd_2 = Add(
   A=Add(
    A=MaterialFunctionCall!(TextureResolution=@n2) {materialFunction=/Engine/Functions/Engine_MaterialFunctions02/Texturing/ScreenAlignedPixelToPixelUVs}.Result
    B=FunctionInput() {bUsePreviewValueAsDefault=true, inputName=Offset, inputType=FunctionInput_Vector2}
   )
   B=ComponentMask(None=@n3) {r=true}
  )
  n4 = Custom!(p=@MaterialExpressionAdd_2) {
   code="((uint)(p.x) + 2 * (uint)(p.y)) % 5"
   outputType=CMOT_Float1
   inputs [{
     inputName=p
     input {expression=&self:MaterialExpressionAdd_2}
    }]
  }
  n5 = Constant2Vector() {g=64, r=64}
  n6 = TextureSample(UVs=Add(
    A=MaterialFunctionCall(TextureResolution=@n5) {materialFunction=&ScreenAlignedPixelToPixelUVs}.Result
    B=FunctionInput() {bUsePreviewValueAsDefault=true, inputName=Offset2, inputType=FunctionInput_Vector2, sortPriority=3}
   )) {samplerType=SAMPLERTYPE_LinearGrayscale, texture=/Engine/EngineMaterials/Good64x64TilingNoiseHighFreq}
  n7 = FunctionInput() {
   bUsePreviewValueAsDefault=true
   description="0=no random 1:random(default)"
   inputName=Random
   inputType=FunctionInput_Scalar
   previewValue=<1 0 0 1>
  }
  n8 = Divide(A=Add(A=@n4, B=Multiply(A=@n6.R, B=@n7)) {constB=0}) {constB=6}
  FunctionOutput(None=Add(A=Add(A=@n1, B=@n8) {constB=0.5}) {constB=-0.5})
 ]
}
