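# Tree material function that optionally thickens spline-built branches, toggled by the Use Branch Thickening switch on incoming material attributes.
# Two MS_CanopyCreatorMeshExpansion calls supply expanded world position offset, normals, radius and masks separating thickened splines, knots and regular mesh.
# Those masks blend branch diffuse, color tint, normal and opacity into the inputs.
# Optional pixel depth offset combines height-scaled radius with dithered knots, zeroed at Low and Medium quality.
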MaterialFunction /Figment_S07/LM_7_00/Packages/Fortnite_UniqueMaterials_1/MaterialFunctions/Trees/MF_TreeSplineThickenBranches
graph {nodes [
  n1 = TextureSampleParameter2D() {
   group=Branch Thickening
   parameterName=Branch Thicken Diffuse Map
   texture=/Figment_S07/LM_7_00/Environments/Props/Foliage/Textures/SplineBranch_01_D
  }
  n2 = FunctionInput() {inputType=FunctionInput_MaterialAttributes}
  n3 = BreakMaterialAttributes!(Attr=@n2)
  n4 = TextureSampleParameter2D() {
   group=Branch Thickening
   parameterName=Branch Thicken Normal
   samplerType=SAMPLERTYPE_Normal
   texture=/Figment_S07/LM_7_00/Environments/Props/Foliage/Textures/SplineBranch_01_N
  }
  n5 = MaterialFunctionCall!(WorldPosition=WorldPosition().XYZ, Normal Map=@n4.RGB) {materialFunction=/Engine/Functions/Engine_MaterialFunctions02/MAXScripts/MS_CanopyCreatorMeshExpansion}
  n6 = Add(A=@n5."white = knots black = everything else", B=OneMinus(None=@n5."Black=Spline Thicken White=Normal"))
  n7 = Multiply(
   A=LinearInterpolate(A=@n1.RGB, B=@n3.BaseColor, Alpha=@n5."Black=Spline Thicken White=Normal")
   B=LinearInterpolate(B=VectorParameter() {group=Branch Thickening, parameterName=Branch Color, defaultValue=<rgba 0.56 0.56 0.56 1>}.RGB, Alpha=Clamp(None=@n6)) {constA=1}
  )
  n8 = TextureSampleParameter2D() {
   group=Branch Thickening
   parameterName=Branch Thicken Mask_R_Height_G_Mask
   samplerType=SAMPLERTYPE_LinearColor
   texture=/Figment_S07/LM_7_00/Environments/Props/Foliage/Textures/SplineBranch_01_R_Height_G_AO
  }
  n9 = TextureSampleParameter2D() {
   group=Branch Thickening
   parameterName=Branch Thicken Normal_1
   samplerType=SAMPLERTYPE_Normal
   texture=&SplineBranch_01_N
  }
  n10 = MaterialFunctionCall(WorldPosition=Add(B=WorldPosition().XYZ), Normal Map=@n9.RGB) {materialFunction=&MS_CanopyCreatorMeshExpansion}
  n11 = LinearInterpolate(A=@n8.G, B=@n3.OpacityMask, Alpha=@n10."Black=Spline Thicken White=Normal")
  n12 = LinearInterpolate(A=@n5.Normal, B=@n3.Normal, Alpha=Ceil(None=@n10."Black=Spline Thicken White=Normal"))
  n13 = Add(A=@n3.WorldPositionOffset, B=@n10.WorldPositionOffset)
  n14 = Multiply(A=@n10.Branch Radius, B=@n8.R) {constB=40}
  n15 = MaterialFunctionCall!(Alpha Threshold, Offset, Random, Offset2=TextureCoordinate() {coordinateIndex=1, uTiling=150.358, vTiling=150.358}) {materialFunction=/Figment_S07/LM_7_00/Packages/Fortress_SharedMaterials/Base_Material_Functions/DitherTemporalAA_Fortnite}
  n16 = ScalarParameter() {defaultValue=7.3166, group=Branch Thickening, parameterName=Dither Distance}
  n17 = Add(A=@n3.PixelDepthOffset, B=Multiply(A=@n10."white = knots black = everything else", B=Multiply(A=@n15.Result, B=@n16)))
  n18 = Constant()
  n19 = StaticSwitchParameter(
   True=QualitySwitch(Default=LinearInterpolate(A=@n14, B=@n17, Alpha=@n10."Black=Spline Thicken White=Normal") {desc=PDO}, Low=@n18, Medium=@n18)
   False=@n18
  ) {group=a_Optimizations, parameterName="Use Pixel Depth Offset (PDO)"}
  FunctionOutput(None=StaticSwitchParameter(
    True=MakeMaterialAttributes!(
     BaseColor=@n7
     Metallic=@n3.Metallic
     Specular=@n3.Metallic
     Roughness=@n3.Metallic
     Anisotropy
     EmissiveColor=@n3.Metallic
     Opacity=@n3.Metallic
     OpacityMask=@n11
     Normal=@n12
     Tangent
     WorldPositionOffset=@n13
     SubsurfaceColor=@n3.Metallic
     ClearCoat=@n3.Metallic
     ClearCoatRoughness=@n3.Metallic
     AmbientOcclusion=@n3.Metallic
     Refraction=@n3.Metallic
     CustomizedUVs_0=@n3.Metallic
     CustomizedUVs_1=@n3.Metallic
     CustomizedUVs_2=@n3.Metallic
     CustomizedUVs_3=@n3.Metallic
     CustomizedUVs_4=@n3.Metallic
     CustomizedUVs_5=@n3.Metallic
     CustomizedUVs_6=@n3.Metallic
     CustomizedUVs_7=@n3.Metallic
     PixelDepthOffset=@n19
     ShadingModel
     Displacement
    )
    False=@n2
   ) {group=Branch Thickening, parameterName=Use Branch Thickening})
 ]}
