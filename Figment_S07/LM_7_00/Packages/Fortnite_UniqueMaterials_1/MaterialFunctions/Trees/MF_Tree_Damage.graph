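# Material function for tree damage that, as wired, passes its material attributes straight to the output, leaving all damage logic unconnected.
# That logic tints a cropped trunk damage texture and lerps it into bark by damage percentage, masking out leaves.
# A static switch can instead blend in a damage reveal texture, which also alters opacity mask.
# Separate If and SphereMask nodes derive a bark-destruction mask, also feeding nothing.
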MaterialFunction /Figment_S07/LM_7_00/Packages/Fortnite_UniqueMaterials_1/MaterialFunctions/Trees/MF_Tree_Damage
graph {nodes [
  n1 = FunctionInput() {inputType=FunctionInput_MaterialAttributes}
  FunctionOutput(None=@n1)
  n2 = MaterialFunctionCall!(Material=@n1) {materialFunction=/Engine/Functions/MaterialLayerFunctions/MatLayerBlend_BreakBaseColor}
  n3 = TextureObjectParameter() {group=Damage, parameterName=Trunk Damage Texture, texture=/Figment_S07/LM_7_00/Packages/User_Built_Props/Textures/wood6}
  n4 = VectorParameter() {group=Damage, parameterName=Trunk damage texture tiling, defaultValue=<rgba 0 0 0.3 0.3>}
  n5 = ComponentMask(None=@n4.RGB) {g=true, r=true}
  n6 = AppendVector(A=@n4.B, B=@n4.B)
  n7 = FunctionInput() {bUsePreviewValueAsDefault=true, inputName=Mask for Leafs, inputType=FunctionInput_Scalar, sortPriority=4, previewValue=<1 0 0 1>}
  n8 = FunctionInput(Preview=ScalarParameter() {parameterName=Code_CustomDamagePercentage}) {bUsePreviewValueAsDefault=true, inputName=CODE_DamagePercentage, inputType=FunctionInput_Scalar, sortPriority=8, previewValue=<1 0 0 1>}
  n9 = LinearInterpolate(
   A=@n2.BaseColor
   B=Multiply(
    A=MaterialFunctionCall!(UVs, TextureIn=@n3, UpperLeftCorner=@n5, LowerRightCorner=@n6, ExportFloat4) {materialFunction=/Engine/Functions/Engine_MaterialFunctions01/Texturing/TextureCropping}.Cropped
    B=VectorParameter() {group=Damage, parameterName=BarkColorDamage, defaultValue=<rgba 1 0.706 0.3688 1>}.RGB
   ) {desc=colorize}
   Alpha=Multiply(A=OneMinus(None=@n7), B=@n8)
  ) {desc=lerp in bark}
  LinearInterpolate(A=@n9, B=@n2.BaseColor, Alpha=@n7) {desc=leaves}
  n10 = OneMinus(None=Constant())
  n11 = FunctionInput() {bUsePreviewValueAsDefault=true, inputName=DamageMask, inputType=FunctionInput_Scalar, sortPriority=5, previewValue=<1 0 0 1>}
  n12 = Constant() {r=1}
  n13 = If(A=@n10, B=@n11, "A > B"=@n12, "A == B"=@n12, "A < B"=Constant()) {desc=base interpolation, equalsThreshold=0.001}
  OneMinus(None=@n13)
  n14 = OneMinus(None=@n13)
  n15 = SphereMask(A=@n10, B=@n11, Radius=Constant() {r=0.05}) {hardnessPercent=0}
  StaticSwitchParameter(True=Clamp(None=Subtract(A=@n14, B=@n15)), False=Constant()) {defaultValue=true, group=Damage, parameterName=DestroyBark}
  n16 = TextureSampleParameter2D() {group=Textures, parameterName=Damage reveal texture, texture=/Figment_S07/LM_7_00/Environments/Props/Trees/Textures/PineBranches_D}
  n17 = MaterialFunctionCall!(
   Material=@n1
   NewBaseColor=StaticSwitchParameter(True=LinearInterpolate(A=@n2.BaseColor, B=@n16.RGB, Alpha=@n8) {desc=lerp damage reveal texture}, False=@n9) {group=Damage, parameterName=Use Damage Reveal Texture}
   Mask
  ) {materialFunction=/Engine/Functions/MaterialLayerFunctions/MatLayerBlend_OverrideBaseColor}
  n18 = Multiply(A=@n16.A, B=BreakMaterialAttributes!(Attr=@n1).OpacityMask)
  StaticSwitchParameter(
   True=StaticSwitchParameter(
    True=MaterialFunctionCall!(Material=@n17.Blended Material, NewOpacityMask=@n18, Mask=@n8) {materialFunction=/Engine/Functions/MaterialLayerFunctions/MatLayerBlend_OverrideOpacityMask}.Blended Material
    False=@n17.Blended Material
   ) {group=Damage, parameterName=Use Damage Reveal Texture}
   False=@n1
  ) {defaultValue=true, desc=Base Color, group=Damage, parameterName=Use Damage Effects}
 ]}
