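# Material function for trees that adds height-based Pixel Depth Offset to incoming material attributes, with a mask input for leaves.
# Height is scaled by a parameter and the leaf mask, then jittered using DitherTemporalAA_Fortnite fed by tiled UV1 coordinates.
# A quality switch zeroes the offset on Low and Medium, and a static switch can disable it entirely.
# The result is written to PixelDepthOffset in the rebuilt attributes.
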MaterialFunction /Figment_S07/LM_7_00/Packages/Fortnite_UniqueMaterials_1/MaterialFunctions/Trees/MF_Tree_PixelDepthOffset
graph {nodes [
  n1 = BreakMaterialAttributes!(Attr=FunctionInput() {inputType=FunctionInput_MaterialAttributes})
  n2 = FunctionInput() {bUsePreviewValueAsDefault=true, inputName=Mask for Leafs, inputType=FunctionInput_Scalar, sortPriority=3, previewValue=<1 0 0 1>}
  n3 = Multiply(
   A=@n2
   B=Multiply(
    A=FunctionInput() {bUsePreviewValueAsDefault=true, inputName=Height, inputType=FunctionInput_Scalar, sortPriority=3}
    B=ScalarParameter() {defaultValue=5, group=PX_Pixel Depth Offset, parameterName=Height Map Pixel Depth offset}
   ) {constB=5}
  )
  n4 = Multiply(
   A=MaterialFunctionCall!(Alpha Threshold, Offset, Random, Offset2=TextureCoordinate() {coordinateIndex=1, uTiling=150.358, vTiling=150.358}) {materialFunction=/Figment_S07/LM_7_00/Packages/Fortress_SharedMaterials/Base_Material_Functions/DitherTemporalAA_Fortnite}.Result
   B=ScalarParameter() {defaultValue=10, group=PX_Pixel Depth Offset, parameterName=Height Map Pixel Depth Jitter Offset}
  ) {constB=10}
  n5 = Constant()
  n6 = QualitySwitch(Default=Add(A=@n3, B=Multiply(A=@n3, B=@n4)), Low=@n5, Medium=@n5)
  FunctionOutput(None=MakeMaterialAttributes!(
    BaseColor=@n1.BaseColor
    Metallic=@n1.BaseColor
    Specular=@n1.BaseColor
    Roughness=@n1.BaseColor
    Anisotropy
    EmissiveColor=@n1.BaseColor
    Opacity=@n1.BaseColor
    OpacityMask=@n1.BaseColor
    Normal=@n1.BaseColor
    Tangent
    WorldPositionOffset=@n1.BaseColor
    SubsurfaceColor=@n1.BaseColor
    ClearCoat=@n1.BaseColor
    ClearCoatRoughness=@n1.BaseColor
    AmbientOcclusion=@n1.BaseColor
    Refraction=@n1.BaseColor
    CustomizedUVs_0=@n1.BaseColor
    CustomizedUVs_1=@n1.BaseColor
    CustomizedUVs_2=@n1.BaseColor
    CustomizedUVs_3=@n1.BaseColor
    CustomizedUVs_4=@n1.BaseColor
    CustomizedUVs_5=@n1.BaseColor
    CustomizedUVs_6=@n1.BaseColor
    CustomizedUVs_7=@n1.BaseColor
    PixelDepthOffset=StaticSwitchParameter(True=@n6, False=@n5) {group=a_Optimizations, parameterName="Use Pixel Depth Offset (PDO)"}
    ShadingModel
    Displacement
   ))
 ]}
