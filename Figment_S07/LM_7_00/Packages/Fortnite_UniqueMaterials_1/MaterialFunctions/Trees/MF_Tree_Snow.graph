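# Material function that adds snow to tree materials by altering only the BaseColor of the incoming material attributes.
# When UseSnow is on, it lerps toward SnowColor_1 using contrast-adjusted Normal blue multiplied by contrast-adjusted BaseColor red.
# When off, it still lerps toward light grey using MF_Athena_Holiday_SnowFall's simple mask, sampled at actor or pixel position by static switch.
# A SnowUp Min/Max blend is left unconnected.
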MaterialFunction /Figment_S07/LM_7_00/Packages/Fortnite_UniqueMaterials_1/MaterialFunctions/Trees/MF_Tree_Snow
graph {nodes [
  Constant3Vector()
  Constant()
  n1 = GetMaterialAttributes(None=FunctionInput() {inputType=FunctionInput_MaterialAttributes}) {attributeGetTypes [#1, #2]}
  n2 = VectorParameter() {group=Snow, parameterName=SnowColor_1, defaultValue=<rgba 0.7262 0.7933 0.83 1>}
  n3 = LinearInterpolate(
   A=ScalarParameter() {defaultValue=-0.0381, group=Snow, parameterName=SnowNormalContrast1}
   B=ScalarParameter() {defaultValue=1.4108, group=Snow, parameterName=SnowNormalContrast2}
   Alpha=ComponentMask(None=@n1.Normal) {b=true}
  )
  n4 = LinearInterpolate(
   A=ScalarParameter() {group=Snow, parameterName=SnowDiffuseContrast1}
   B=ScalarParameter() {defaultValue=13.0863, group=Snow, parameterName=SnowDiffuseContrast2}
   Alpha=ComponentMask(None=@n1.BaseColor) {r=true}
  )
  n5 = LinearInterpolate(A=@n1.BaseColor, B=@n2.RGB, Alpha=Multiply(A=Clamp(None=@n3), B=Clamp(None=@n4)))
  n6 = ActorPositionWS()
  n7 = WorldPosition()
  n8 = StaticBoolParameter() {defaultValue=true, parameterName=UseActorPositionForSnow}
  n9 = LinearInterpolate(
   A=@n1.BaseColor
   Alpha=MaterialFunctionCall!(
    UseHeightLerp
    UV_WorldMask
    Use_ExtraSnowMask
    LandscapeBlend_Height
    Landscape_RockMask
    Snow_Mask
    Simple_UseWorldZNormal=StaticBool() {value=true}
    Simple_WorldNormal=@n1.Normal
    WorldPosition=StaticSwitch(True=@n6, False=@n7.XYZ, Value=@n8)
   ) {materialFunction=/Figment_S07/LM_7_00/Athena/Environments/Landscape/MaterialFunctions/Snow/MF_Athena_Holiday_SnowFall}.Simple_SnowMask
  ) {constB=0.7}
  FunctionOutput(None=StaticSwitchParameter(
    True=SetMaterialAttributes!(MaterialAttributes=@n1.MaterialAttributes, BaseColor=@n5) {attributeSetTypes [#1]}
    False=SetMaterialAttributes(MaterialAttributes=@n1.MaterialAttributes, BaseColor=@n9) {attributeSetTypes [#1]}
   ) {parameterName=UseSnow})
  n10 = ScalarParameter() {defaultValue=-0.5, parameterName=SnowUp_Min}
  n11 = ScalarParameter() {defaultValue=1, parameterName=SnowUp_Max}
  n12 = ComponentMask(None=@n1.Normal) {b=true}
  LinearInterpolate(A=@n1.BaseColor, B=@n2.RGB, Alpha=Clamp(None=LinearInterpolate(A=@n10, B=@n11, Alpha=@n12)))
  ObjectPositionWS()
 ]}
